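# M_Wood_Chips is a lit material for wood chip effects, flagged for use on Cascade and Niagara mesh particles and with Nanite.
# The RGB of the T_Wood_Chips texture feeds the diffuse input of ConvertFromDiffSpec, which converts diffuse/specular-style color to PBR values.
# A Roughness scalar parameter, fully rough by default, is exposed for material instances.
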Material /Game/Effects/Fort_Effects/Materials/Wood/M_Wood_Chips
graph {
 properties {
  bUsedWithMeshParticles=true
  bUsedWithNanite=true
  bUsedWithNiagaraMeshParticles=true
  pixelDepthOffsetMode=PDOM_Legacy
  usedShadingModels=Default Lit
 }
 nodes [
  MaterialFunctionCall(DiffuseColor=TextureSample() {texture=/Game/Effects/Fort_Effects/Textures/Wood/T_Wood_Chips}.RGB) {materialFunction=&ConvertFromDiffSpec}
  ScalarParameter() {defaultValue=1, parameterName=Roughness}
 ]
}
